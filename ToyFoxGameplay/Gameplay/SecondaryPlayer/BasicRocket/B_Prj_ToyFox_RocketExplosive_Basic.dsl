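; Basic ToyFox secondary-player rocket projectile that handles the cosmetic side of explosion and an impact telegraph marker.
; SpawnImpactTelegraph, gated by a hotfix flag and limited to the local instigating player, waits a tick to spare the costly spawn frame, then async-loads and spawns the telegraph.
; On explosion it hides the mesh, deactivates the Niagara trail to fade naturally, and destroys the telegraph.
; EndPlay also destroys the telegraph as cleanup.
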(event EventOnExploded
  (params (HitActors :type Array%20of%20Actor%20Object%20References) (HitResults :type Array%20of%20Hit%20Result%20Structures))
  (bind _niagara_trail (Variables|Default|GetNiagara_Trail))
  ;[0.15 0.0525 0.1492 0.5] OnExploded: Hiding the mesh but letting the niagara trail die out naturally.
  (|CallParentFunction HitActors HitResults)
  (Networking|CanExecuteCosmeticEvents
    (:True
      (Rendering|SetActorHiddenInGame self false)
      (Development|SetHiddeninGame (Variables|Default|GetStaticMesh) true false)
      (Development|SetHiddeninGame _niagara_trail false false)
      (Components|Activation|Deactivate _niagara_trail)
      (bind _impacttelegraphactor (Variables|Transient|GetImpactTelegraphActor)
        (:then
          (Actor|DestroyActor _impacttelegraphactor))
        (:else)))))

(event Custom|SpawnImpactTelegraph
  (params (Location :type Vector) (Rotation :type Rotator))
  (bind _returnvalue (Projectile|GetInstigatorPlayerController))
  ;[0.007622 0.09873 0.15 0.5] SpawnImactTelegraph: Spawn the impact telegraph at the input location.
  (Networking|CanExecuteCosmeticEvents
    (:True
      (Utilities|IsValid _returnvalue
        (:Is%20Valid
          (if (Pawn|IsLocalPlayerController _returnvalue)
            (if (ScalableFloat|GetValueasBool (Variables|Hotfix|GetHotfix_ShowImpactTelegraph))
              (Utilities|FlowControl|DelayUntilNextTick) ; Can wait a frame for perf.\nThis is called when projectile has just spawned (expensive frame)
              (bind _class (Utilities|AsyncLoadClassAsset (Variables|Settings|GetTelegraphActorClass))
                (:Completed
                  (bind _asactor (Utilities|Casting|CastToActorClass _class))
                  (bind _returnvalue_2 (Game|SpawnActorfromClass _asactor (Math|Transform|MakeTransform Location Rotation) "AlwaysSpawn" "MultiplyWithRoot" (Projectile|GetInstigatorPlayerController) (Game|Player|GetPlayerPawn (Projectile|GetInstigatorPlayerController))))
                  (Variables|Transient|SetImpactTelegraphActor _returnvalue_2))))))))))

(event EventEndPlay
  (params (EndPlayReason :type EEndPlayReason%20Enum :default "Destroyed"))
  ;[0.15 0.1015 0 0.5] EndPlay: Clean up.
  (Utilities|FlowControl|Sequence
    (:then_0
      (bind _impacttelegraphactor (Variables|Transient|GetImpactTelegraphActor)
        (:then
          (Actor|DestroyActor _impacttelegraphactor))
        (:else)))
    (:then_1
      (|CallParentFunction EndPlayReason))))

(fn ConstructionScript
  (|CallParentFunction))
