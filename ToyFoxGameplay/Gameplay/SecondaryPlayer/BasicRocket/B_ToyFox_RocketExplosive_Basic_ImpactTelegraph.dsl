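; B_ToyFox_RocketExplosive_Basic_ImpactTelegraph is an impact telegraph actor for the ToyFox secondary player's basic explosive rocket.
; On BeginPlay, after the parent call, it checks whether cosmetic events can run.
; If so, it hides itself for every local controller except the instigator's, so splitscreen players do not see each other's telegraphs.
; The construction script is empty.
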(event EventBeginPlay
  ;[0 0.15 0.06605 0.5] BeginPlay: This actor should be hidden for local controllers that are not the instigator. For splitscreen.
  (|CallParentFunction)
  (Networking|CanExecuteCosmeticEvents
    (:True
      (Fortnite|SetActorHiddenforNonInstigatorLocalPlayerControllers self))))

(fn ConstructionScript)
