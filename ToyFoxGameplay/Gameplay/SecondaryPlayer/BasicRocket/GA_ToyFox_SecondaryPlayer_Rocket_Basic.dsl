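; Gameplay ability for the ToyFox secondary pilot that fires a basic rocket from the giant pawn toward the camera's aim point.
; A camera line trace picks the target; the giant's rocket component spawns the projectile from a camera-relative location, shifted by giant velocity to keep it on screen.
; A successful spawn triggers the fire cue, effect and pilot event, then commits cooldown and consumes ammo.
; Activation requires ammo; AI noise is emitted, and cooldown failures are hidden from the HUD.
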(event Ability|EventActivateAbility
  (bind _returnvalue (CallFunction|GetAvatarasFortPlayerPawn))
  (bind _giantpawn (Variables|Transient|GetGiantPawn))
  (bind _returnvalue_2 (Actor|GetComponentbyClass _giantpawn "/ToyFoxGameplay/Gameplay/SecondaryPlayer/B_ToyFoxGiantPawnRocketComponent.B_ToyFoxGiantPawnRocketComponent_C"))
  (bind _returnvalue_3 (Actor|GetComponentbyClass _returnvalue "/ToyFoxGameplay/Gameplay/PawnComp/B_ToyFoxPilotPawnComponent_Parent.B_ToyFoxPilotPawnComponent_Parent_C"))
  ;[0 0.15 0.01742 0.5] AcitvateAbility: Fire projectile at the aimed direction.
  (Utilities|FlowControl|Sequence
    (:then_0
      (Variables|Transient|SetGiantPawn (Giant|GetGiantPawn (Actor|GetComponentbyClass (Game|GetGameState) "/ToyFoxGameplay/Gameplay/B_ToyFoxGameStateComponent.B_ToyFoxGameStateComponent_C")))
      (Variables|Transient|SetbProjectileSpawned false))
    (:then_1
      (Utilities|IsValid (Variables|Transient|GetGiantPawn)
        (:Is%20Valid
          (bind _outfortplayerpawn (CallFunction|GetAvatarasValidFortPlayerPawn)
            (:ValidFortPlayerPawn
              (bind _returnvalue_4 (Fortnite|GetFortPlayerControllerfromActor _outfortplayerpawn))
              (Utilities|IsValid _returnvalue_4
                (:Is%20Valid
                  (bind _playercameramanager (Variables|/Script/Engine.PlayerController|GetPlayerCameraManager _returnvalue_4)
                    (:then
                      (bind (_outhit _returnvalue_5) (Collision|LineTraceByChannel (Camera|GetCameraLocation _playercameramanager) (+ (Camera|GetCameraLocation _playercameramanager) (* (Math|Vector|GetForwardVector (Camera|GetCameraRotation _playercameramanager)) 1e+05)) "TraceTypeQuery8" true (Utilities|Array|MakeArray (CallFunction|GetAvatarasFortPlayerPawn) (Variables|Transient|GetGiantPawn)) "None" true (select (Networking|IsDedicatedServer) "(R=0.139049,G=1.000000,B=0.000000,A=1.000000)" "(R=0.000000,G=0.777072,B=1.000000,A=1.000000)") "(R=0.000000,G=1.000000,B=0.000000,A=1.000000)" 8.0))
                      (Utilities|IsValid _returnvalue_2
                        (:Is%20Valid
                          (bind _projectile (CallFunction|RequestSpawnRocketProjectile _returnvalue_2 (Math|Vector|SelectVector (.ImpactPoint (Collision|BreakHitResult _outhit)) (.TraceEnd (Collision|BreakHitResult _outhit)) _returnvalue_5) (Utilities|Operators|Add (Camera|GetCameraLocation _playercameramanager) (* (Math|Vector|GetUpVector (Camera|GetCameraRotation _playercameramanager)) (ScalableFloat|GetCachedScalableFloatValue (Variables|Hotfix|GetHotfix_ProjectileSpawnLocationZOffsetFromCameraLoc))) (CallFunction|GetSpawnLocationOffsetBasedonGiantVelocity)) (.ImpactNormal (Collision|BreakHitResult _outhit))))
                          (Utilities|IsValid _projectile
                            (:Is%20Valid
                              (Ability|ExecuteGameplayCueOnOwner (Utilities|Struct|BreakGameplayCueTag (Variables|GameplayCue|GetGameplayCueTag_Fire)) (GameplayEffects|MakeEffectContext (Ability|GetAbilitySystemComponentfromActorInfo)))
                              (Variables|Transient|SetbProjectileSpawned true)
                              (Ability|ApplyGameplayEffectToOwner (Variables|Settings|GetRecentlyUsedGameplayEffectClass))
                              (Utilities|IsValid _returnvalue_3
                                (:Is%20Valid
                                  (Class|BToyFoxPilotPawnComponentParent|BroadcastInputActionEventTag _returnvalue_3 "(TagName=\"ToyFox.Event.Ability.Secondary.Rocket.Basic.Fired\")"))))))))
                    (:else)))))))))
    (:then_2
      ;;[0.03689 0.04106 0.15 0.5] Add noise event for AI to react to
      (bind _outfortplayerpawn_2 (CallFunction|GetAvatarasValidFortPlayerPawn)
        (:ValidFortPlayerPawn
          (AI|Perception|MakeNoiseEventatLocation _outfortplayerpawn_2 (ScalableFloat|GetValueatLevel (Variables|Hotfix|GetHotfix_PerceptionRange)) (Transformation|GetActorLocation _outfortplayerpawn_2) "WeaponFiring"))))
    (:then_3
      (if (Variables|Transient|GetbProjectileSpawned)
        (CallFunction|CustomCommitAbility)
        (else
          _))
      (Ability|EndAbility))))

(event Custom|CustomCommitAbility
  ;[0 0.1075 0.15 0.5] CustomCommitAbility: Apply cooldown and consume ammo.
  (Ability|CommitAbilityCooldown false)
  (bind _outfortplayerpawn (CallFunction|GetAvatarasValidFortPlayerPawn)
    (:ValidFortPlayerPawn
      (bind _returnvalue (Actor|GetComponentbyClass _outfortplayerpawn "/ToyFoxGameplay/Gameplay/PawnComp/B_ToyFoxPilotPawnComponent_Secondary.B_ToyFoxPilotPawnComponent_Secondary_C"))
      (Utilities|IsValid _returnvalue
        (:Is%20Valid
          (Rocket|Basic|ConsumeBasicRocketAmmo _returnvalue 1))))))

(fn GetSpawnLocationOffsetBasedOnGiantVelocity
  (bind _giantpawn (Variables|Transient|GetGiantPawn))
  ;[0.02334 0.1351 0.15 0.5] Offset the spawn location of the projectile based on the velocity of the giant so that the rocket spawns on the player's screen.
  (Utilities|FlowControl|Sequence
    (:then_0
      (Utilities|IsValid _giantpawn
        (:Is%20Valid
          (LocalVariables|SetGiantParent _giantpawn))
        (:Is%20Not%20Valid
          (bind _returnvalue (Fortnite|Athena|GameState|GetGameStateAthena))
          (Utilities|IsValid _returnvalue
            (:Is%20Valid
              (bind _returnvalue_2 (Actor|GetComponentbyClass _returnvalue "/ToyFoxGameplay/Gameplay/B_ToyFoxGameStateComponent.B_ToyFoxGameStateComponent_C"))
              (Utilities|IsValid _returnvalue_2
                (:Is%20Valid
                  (LocalVariables|SetGiantParent (Giant|GetGiantPawn _returnvalue_2))))))))
      (bind _giantparent (LocalVariables|GetGiantParent)
        (:then
          (return (Utilities|Operators|Multiply (Math|Conversions|ToVector2D(Vector) (Transformation|GetVelocity _giantparent)) (Utilities|Time|GetWorldDeltaSeconds) (ScalableFloat|GetCachedScalableFloatValue (Variables|Hotfix|GetHotfix_PrjSpawnLocationForwardXYOffsetVelocityBasedMult)))))
        (:else)))
    (:then_1
      (return "0, 0, 0"))))

(fn CanActivateAbility
  (params (ActorInfo :type Gameplay%20Ability%20Actor%20Info%20Structure) (Handle :type Gameplay%20Ability%20Spec%20Handle%20Structure))
  (bind (_owneractor _avataractor) (Utilities|Struct|BreakGameplayAbilityActorInfo ActorInfo))
  (Utilities|FlowControl|Sequence
    (:then_0
      (bind _asfort_player_pawn (Utilities|Casting|CastToFortPlayerPawn _avataractor))
      (bind _returnvalue (Actor|GetComponentbyClass _asfort_player_pawn "/ToyFoxGameplay/Gameplay/PawnComp/B_ToyFoxPilotPawnComponent_Secondary.B_ToyFoxPilotPawnComponent_Secondary_C"))
      (Utilities|IsValid _returnvalue
        (:Is%20Valid
          (if (> (Rocket|Basic|GetCurrentBasicRocketAmmoCount _returnvalue) 0)
            (return 0 true)
            (else
              (Audio|PlaySound2D (Variables|Sound|GetNoAmmoSoundCue))
              (return "(GameplayTags=((TagName=\"Activate.Failed.ToyFox.Rocket.Basic.NoAmmo\")))" false))))))
    (:then_1
      (return 0 false))))

(fn OverrideFailedReason
  (params (FailedReason :type Gameplay%20Tag%20Container%20Structure%20%28by%20ref%29))
  (bind _overridenfailedreason (|CallParentFunction FailedReason))
  (GameplayTags|RemoveGameplayTag _overridenfailedreason "(TagName=\"Activate.Failed.Cooldown\")" :!class /Script/GameplayTags.BlueprintGameplayTagLibrary) ; Remove cooldown fail reason to prevent HUD warning screen popup.
  (return _overridenfailedreason))

(fn RequestSpawnRocketProjectile
  (params (RocketComponent :type B%20Toy%20Fox%20Giant%20Pawn%20Rocket%20Component%20Object%20Reference) (TargetDestinationLocation :type Vector) (SpawnLoc :type Vector) (ImpactNormal :type Vector))
  (bind _explosiverocketprojectile (Class|BToyFoxGiantPawnRocketComponent|RequestSpawnBasicRocket RocketComponent TargetDestinationLocation (CallFunction|GetAvatarasFortPlayerPawn) SpawnLoc ImpactNormal false))
  (return _explosiverocketprojectile))
